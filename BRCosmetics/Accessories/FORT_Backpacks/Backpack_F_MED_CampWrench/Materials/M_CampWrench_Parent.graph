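# Translucent parent material for the CampWrench back bling, layering an animated, iridescent edge-glow and distortion effect onto the shared 2023 character material functions.
# Panned, warped mask-texture noise, wipes and radial motion combine with curvature, Fresnel and camera-distance masks to build edges, thin-film color and vertex offset.
# Refraction is faked by sampling SceneColor at offset screen UVs and blending in time-of-day-adjusted color, with separate mobile color and opacity paths.
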Material /BRCosmetics/Accessories/FORT_Backpacks/Backpack_F_MED_CampWrench/Materials/M_CampWrench_Parent
graph {
 properties {
  bUseMaterialAttributes=true
  bUsedWithMeshDeformer=true
  bUsedWithMorphTargets=true
  bUsedWithSkeletalMesh=true
  blendMode=BLEND_Translucent
  pixelDepthOffsetMode=PDOM_Legacy
  previewMesh=/BRCosmetics/Accessories/FORT_Backpacks/Backpack_F_MED_CampWrench/Meshes/F_MED_CampWrench_Pack
  shadingModel=MSM_FromMaterialExpression
  translucencyDirectionalLightingIntensity=0.1
  usedShadingModels=Default Lit | Subsurface | Clear Coat | Subsurface Profile | Cloth
 }
 nodes [
  n1 = NamedRerouteUsage()
  Add(A=@n1, B=Constant3Vector() {constant=<rgba 0 0 0.5 0>})
  ChannelMaskParameter(None=NamedRerouteUsage()) {parameterName=EdgesMask}
  ConstantBiasScale(None=NamedRerouteUsage())
  Constant() {r=1}
  n2 = Time()
  n3 = Sine(None=Add(A=NamedRerouteUsage(), B=Multiply(A=@n2) {constB=0.2}))
  Divide(A=Multiply(A=@n3)) {constB=1}
  n4 = NamedRerouteUsage()
  n5 = ScalarParameter() {defaultValue=0.01, parameterName=DistorionAmount}
  n6 = NamedRerouteUsage()
  n7 = SmoothStep(Value=Power(Base=OneMinus(None=@n6)) {constExponent=5})
  n8 = Add(A=ScreenPosition().ViewportUV, B=Multiply(A=Subtract(A=@n4) {constB=0.5}, B=LinearInterpolate(A=@n5, Alpha=@n7) {constB=0}) {constB=0.025})
  n9 = NamedRerouteUsage()
  n10 = Multiply(A=@n9) {constB=0.75}
  n11 = ShadingPathSwitch(
   Default=LinearInterpolate(
    A=SceneColor!(UVs=@n8).RGB
    B=MaterialFunctionCall(Color_Morning=@n9, Color_Day=@n9, Color_Evening=@n10, Color_Night=@n10) {materialFunction=&Color_Time_of_Day}.Result
    Alpha=NamedRerouteUsage()
   )
   Mobile=NamedRerouteUsage()
  ) {desc=Migrated from FeatureLevelSwitch}
  n12 = ComponentMask(None=Add(A=NamedRerouteUsage()) {constB=-0.3}) {r=true}
  LinearInterpolate(A=@n11, B=NamedRerouteUsage(), Alpha=Saturate(None=Round(None=@n12)))
  n13 = MaterialFunctionCall() {materialFunction=&MF_CharacterParent_2023_PreFX}
  MaterialFunctionCall(Material=NamedRerouteUsage(), M=@n13.M, Pre Skinned Local Position=@n13.M, Distance Blend=@n13.M, Fresnel=@n13.M) {materialFunction=&MF_CharacterParent_2023_PostFX}
  n14 = Multiply(A=NamedRerouteUsage(), B=ScalarParameter() {defaultValue=0.3, parameterName=EmissiveIntensity}) {constB=0.1}
  n15 = Multiply(A=@n14) {constB=0.75}
  MaterialFunctionCall(Color_Morning=@n14, Color_Day=@n14, Color_Evening=@n15, Color_Night=@n15) {materialFunction=&Color_Time_of_Day}
  MaterialFunctionCall!(Normal, Light Vector, ViewVector, Glossiness) {materialFunction=/Engine/Functions/Engine_MaterialFunctions01/Lighting/BlinnPhongSpecular}
  MaterialFunctionCall!(UVs, SpecRoughnessMin, SpecRoughnessMax, Metallic) {materialFunction=/Game/Packages/Fortress_SharedMaterials/Base_Material_Functions/SpecularMasks}
  n16 = Multiply(A=NamedRerouteUsage(), B=NamedRerouteUsage())
  n17 = NamedRerouteUsage()
  n18 = Multiply(A=@n16, B=TextureSample(UVs=Add(A=@n3, B=@n17)) {texture=/Game/Global/Textures/Default/Thinfilm/T_ThinFilm_Spectrum_COLOR}.RGB)
  n19 = ScalarParameter() {defaultValue=-0.5, parameterName=FurtacorIntensity}
  n20 = NamedRerouteUsage()
  n21 = Desaturation(
   None=LinearInterpolate(
    A=Desaturation(None=@n18, Fraction=@n19)
    B=VectorParameter() {parameterName=EdgeColor, defaultValue=<rgba 0.2306 0.6423 1 1>}.RGB
    Alpha=NamedRerouteUsage()
   )
   Fraction=Multiply(A=@n20) {constB=-0.3}
  )
  Multiply(A=@n21, B=LinearInterpolate(Alpha=@n20) {constA=0.9, constB=1.5})
  n22 = Multiply(A=NamedRerouteUsage(), B=NamedRerouteUsage())
  n23 = LinearInterpolate(Alpha=Power(Base=@n22)) {constA=0.15, constB=1.5}
  n24 = SmoothStep(Value=@n22) {constMin=0.2}
  n25 = Saturate(None=Multiply(A=Add(A=@n23, B=@n24)))
  n26 = LinearInterpolate(Alpha=NamedRerouteUsage()) {constA=1, constB=0}
  n27 = ScalarParameter() {defaultValue=0.5, parameterName=EdgeFresnelExpo}
  n28 = Fresnel(ExponentIn=LinearInterpolate(A=Multiply(A=@n27) {constB=5}, B=@n27, Alpha=NamedRerouteUsage()), Normal=Transform(None=NamedRerouteUsage())) {baseReflectFraction=0, exponent=4}
  n29 = NamedRerouteUsage()
  n30 = LinearInterpolate(Alpha=Multiply(A=@n29, B=@n29)) {constA=0.8, constB=2}
  n31 = NamedRerouteUsage()
  n32 = ConstantBiasScale(None=NamedRerouteUsage()) {bias=-0.5, scale=2}
  n33 = Multiply(
   A=SphereMask(A=Saturate(None=@n32), B=Constant() {r=1}, Radius=ScalarParameter() {defaultValue=1.3, parameterName=EdgeThickness}) {attenuationRadius=0.7, hardnessPercent=0}
   B=LinearInterpolate(
    A=ScalarParameter() {defaultValue=1.1, parameterName=EdgeIntensityNear}
    B=ScalarParameter() {defaultValue=2.5, parameterName=EdgeIntensityFar}
    Alpha=NamedRerouteUsage()
   )
  )
  n34 = SmoothStep(Value=NamedRerouteUsage()) {constMax=0.3}
  n35 = Add(
   A=Multiply(A=Multiply(A=@n28, B=@n30), B=OneMinus(None=@n31))
   B=LinearInterpolate(A=@n33, B=Multiply(A=@n33, B=@n30), Alpha=OneMinus(None=@n34))
  )
  n36 = NamedRerouteUsage()
  n37 = NamedRerouteUsage()
  n38 = Multiply(A=OneMinus(None=NamedRerouteUsage()), B=ConstantBiasScale(None=Multiply(A=@n36, B=@n37)))
  n39 = Multiply(A=@n35, B=Add(A=@n35, B=Multiply(A=@n38) {constB=0.5}))
  n40 = Saturate(None=Add(A=Multiply(A=@n25, B=@n26), B=Saturate(None=@n39)))
  Multiply(A=@n40, B=LinearInterpolate(Alpha=NamedRerouteUsage()) {constA=0.9, constB=1.1})
  n41 = Constant3Vector() {constant=<rgba 0 1 0 0>}
  n42 = Constant() {r=0.25}
  n43 = Constant()
  n44 = NamedRerouteUsage()
  n45 = ConstantBiasScale(None=NamedRerouteUsage()) {bias=-0.5, scale=2}
  n46 = Add(
   A=Add(A=RotateAboutAxis(NormalizedRotationAxis=@n41, RotationAngle=@n42, PivotPoint=@n43, Position=@n44), B=@n44)
   B=NamedRerouteDeclaration(None=Multiply(A=@n45) {constB=0.25}) {name=EdgeDistortionMask, nodeColor=<rgba 0.4297 1 0 1>}
  )
  n47 = ComponentMask(None=@n46) {b=true, r=true}
  n48 = TextureSample(UVs=Panner(Coordinate=Multiply(A=@n47) {constB=0.5}) {speedX=0.05, speedY=-0.05}) {
   samplerType=SAMPLERTYPE_LinearColor
   texture=/BRCosmetics/Characters/Player/Male/Medium/Bodies/M_MED_EggnogFaucet/Materials/T_EggnogFaucet_mask
  }
  n49 = Add(A=@n48.B, B=Multiply(A=Time()) {constB=0.3})
  n50 = Multiply(
   A=NamedRerouteDeclaration(None=Sine(None=@n49)) {name=AnimatedGaussian_1, nodeColor=<rgba 0 1 0.6016 1>}
   B=LinearInterpolate(Alpha=NamedRerouteUsage()) {constA=0.1, constB=0.3}
  ) {constB=0.1}
  n51 = ConstantBiasScale(None=NamedRerouteUsage()) {bias=-0.5, scale=2}
  n52 = Time()
  n53 = Constant3Vector() {constant=<rgba -1 0 0.5 0>}
  n54 = Add(A=Add(A=Add(A=@n46, B=@n50), B=Multiply(A=@n51) {constB=0.25}), B=Multiply(A=Multiply(A=@n52, B=@n53)) {constB=0.15})
  n55 = Multiply(A=ComponentMask(None=@n54) {b=true, r=true}) {constB=0.75}
  n56 = Constant2Vector() {g=51.2135, r=11.213}
  n57 = TextureSample(UVs=BumpOffset(Coordinate=Add(A=@n55, B=@n56), Height=@n48.G) {heightRatio=0.5}) {
   samplerType=SAMPLERTYPE_LinearColor
   texture=&T_EggnogFaucet_mask
  }
  Multiply(A=@n57.RGB)
  n58 = NamedRerouteUsage()
  Multiply(
   A=MaterialFunctionCall!(
    Normal=Transform(None=@n58)
    Light Vector=SkyAtmosphereLightDirection()
    Roughness=ScalarParameter() {defaultValue=0.5, parameterName=Roughness}
   ) {materialFunction=/Engine/Functions/Engine_MaterialFunctions01/Lighting/GGXSpecular}.Result
   B=VectorParameter() {defaultValue=<rgba 0 0.3143 0.6042 1>}.RGB
  )
  n59 = TextureObject() {
   samplerType=SAMPLERTYPE_LinearColor
   texture=&T_EggnogFaucet_mask
  }
  n60 = PreSkinnedNormal()
  n61 = MaterialFunctionCall(Color=MaterialFunctionCall(TextureObject=@n59, TextureSize=Constant() {r=3}, WorldPosition=@n54, World Space Normal=VertexInterpolator(VS=@n60).PS) {materialFunction=&WorldAlignedTexture}.XYZ Texture) {materialFunction=&SplitComponents}
  NamedRerouteDeclaration(None=@n61.G) {name=DistortedVoron_1, nodeColor=<rgba 1 0 0.9609 1>}
  NamedRerouteDeclaration(None=@n40) {name=Edges, nodeColor=<rgba 0 0.25 1 1>}
  n62 = Subtract(A=Distance(A=ObjectPositionWS(), B=CameraPositionWS()), B=ScalarParameter() {defaultValue=300, parameterName=CameraFadeOffset}) {constB=64}
  n63 = ScalarParameter() {defaultValue=512, parameterName=CameraFadeLength}
  NamedRerouteDeclaration(None=Saturate(None=Divide(A=@n62, B=@n63) {constB=512})) {name=CameraDepthMask, nodeColor=<rgba 0 1 0.2734 1>}
  NamedRerouteDeclaration(None=SphereMask(A=@n1, B=Constant()) {attenuationRadius=1.1, hardnessPercent=0}) {name=CenterMask, nodeColor=<rgba 1 0.0469 0 1>}
  n64 = NamedRerouteUsage()
  n65 = Add(A=Multiply(A=ComponentMask(None=@n64) {b=true}) {constB=0.05}, B=Multiply(A=Time()) {constB=-0.35})
  n66 = Sine(None=@n65)
  n67 = Multiply(A=Add(A=@n65) {constB=0.233}) {constB=1.2}
  n68 = Multiply(A=Power(Base=ConstantBiasScale(None=@n66)) {constExponent=5}, B=ConstantBiasScale(None=Cosine(None=@n67)))
  n69 = NamedRerouteUsage()
  NamedRerouteDeclaration(None=Multiply(A=NamedRerouteDeclaration(None=@n68) {name=Wipe, nodeColor=<rgba 0.9922 1 0 1>}, B=Power(Base=@n69) {constExponent=3})) {name=DetailedWipe, nodeColor=<rgba 1 0.5859 0 1>}
  NamedRerouteDeclaration(None=@n13.Tech Art Mask) {name=InFXMask, nodeColor=<rgba 1 0.1406 0 1>}
  n70 = LinearInterpolate(Alpha=NamedRerouteUsage()) {constA=1, constB=1.5}
  n71 = Transform(None=Constant3Vector() {constant=<rgba 1 0 1 0>}) {transformSourceType=TRANSFORMSOURCE_Local}
  n72 = NamedRerouteUsage()
  n73 = Add(A=Multiply(A=ComponentMask(None=@n72) {b=true}) {constB=-0.5}, B=Multiply(A=Time()) {constB=0.5})
  n74 = Add(A=Multiply(A=Sine(None=@n73)) {constB=0.1}, B=Multiply(A=NamedRerouteUsage()) {constB=0.5}) {constB=0.2}
  n75 = NamedRerouteUsage()
  NamedRerouteDeclaration(None=Multiply(A=Multiply(A=@n70, B=@n71), B=LinearInterpolate(A=@n74, Alpha=@n75))) {name=OuwWPO, nodeColor=<rgba 0.3828 1 0 1>}
  n76 = ConstantBiasScale(None=Multiply(A=NamedRerouteUsage(), B=NamedRerouteUsage())) {bias=-0.5, scale=2}
  n77 = AppendVector(A=DDX(Value=@n76), B=DDY(Value=@n76))
  NamedRerouteDeclaration(None=DeriveNormalZ(InXY=Multiply(A=@n77) {constB=-1})) {name=FakeNormals, nodeColor=<rgba 1 0.2812 0 1>}
  n78 = SetMaterialAttributes!(
   MaterialAttributes=NamedRerouteUsage()
   Base Color=Constant()
   Emissive Color=@n11
   Opacity=ShadingPathSwitch(Default=Constant() {r=1}, Mobile=NamedRerouteUsage()) {desc=Migrated from FeatureLevelSwitch}
   World Position Offset=NamedRerouteUsage()
  ) {attributeSetTypes [
    #1
    #2
    #3
    #4
   ]}
  NamedRerouteDeclaration(None=@n78) {name=OutMaterial, nodeColor=<rgba 1 0.2109 0 1>}
  n79 = NamedRerouteDeclaration(None=@n57.G) {name=DistortedVoron, nodeColor=<rgba 1 0 0.9609 1>}
  n80 = Multiply(A=Time()) {constB=0.5}
  NamedRerouteDeclaration(None=Sine(None=Add(A=@n79, B=@n80))) {name=DistortedVoronAnim, nodeColor=<rgba 0 0.7422 1 1>}
  n81 = TextureSample(UVs=Panner(Coordinate=Multiply(A=@n47) {constB=0.2}) {speedX=0.15, speedY=0.01}) {
   samplerType=SAMPLERTYPE_LinearColor
   texture=&T_EggnogFaucet_mask
  }
  NamedRerouteDeclaration(None=@n81.B) {name=LowFreqPerlin, nodeColor=<rgba 1 0.75 0 1>}
  n82 = MaterialFunctionCall(Vector or UVs=@n47) {materialFunction=&VectorToRadialValue}
  n83 = TextureSample(UVs=Panner(Coordinate=@n82.Radial Coordinates) {speedX=0.05, speedY=-0.25}) {
   samplerType=SAMPLERTYPE_LinearColor
   texture=&T_EggnogFaucet_mask
  }
  NamedRerouteDeclaration(None=Multiply(A=@n83.B, B=@n82.Linear Distance)) {name=RadialBlast, nodeColor=<rgba 0.4297 1 0 1>}
  n84 = DotProduct(A=Transform(None=Constant3Vector() {constant=<rgba 0 1 0 0>}) {transformSourceType=TRANSFORMSOURCE_Local}, B=CameraVectorWS())
  NamedRerouteDeclaration(None=Power(Base=Abs(None=@n84)) {constExponent=4}) {name=FresnelModulator, nodeColor=<rgba 0 1 0.9062 1>}
  NamedRerouteDeclaration(None=@n61.B) {name=DistortedGaussian_1, nodeColor=<rgba 1 0 0.1641 1>}
  NamedRerouteDeclaration(None=GetMaterialAttributes(None=NamedRerouteDeclaration(None=@n13.Result) {name=InMaterial, nodeColor=<rgba 0 1 0.5313 1>}) {attributeGetTypes [#5]}.Normal) {name=InNormal, nodeColor=<rgba 0 0.7187 1 1>}
  n85 = Multiply(A=Add(A=LocalPosition() {localOrigin=InstancePreSkinning}.XYZ, B=Constant3Vector() {constant=<rgba 0 0 6 0>})) {constB=0.1}
  NamedRerouteDeclaration(None=ShaderStageSwitch(PixelShader=VertexInterpolator(VS=@n85).PS, VertexShader=@n85)) {name=PreSkinnedPosition, nodeColor=<rgba 0.6406 1 0 1>}
  NamedRerouteDeclaration(None=@n21) {name=OutEmissive, nodeColor=<rgba 0 1 0.7188 1>}
  NamedRerouteDeclaration(None=@n57.B) {name=DistortedGaussian, nodeColor=<rgba 1 0 0.1641 1>}
  n86 = TextureSample() {
   samplerType=SAMPLERTYPE_LinearColor
   texture=/BRCosmetics/Accessories/FORT_Backpacks/Backpack_F_MED_CampWrench/Textures/T_CampWrench_Backpack_Curvature
  }
  OneMinus(None=NamedRerouteDeclaration(None=@n86.G) {name=EdgesMask, nodeColor=<rgba 0 0.0625 1 1>})
  Power(Base=@n16)
 ]
}
